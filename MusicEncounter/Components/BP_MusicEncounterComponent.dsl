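; Actor component that links its owner to a music encounter and relays gameplay events to it.
; On cosmetic-capable BeginPlay, owner tag checks gate an async request to the GameState's MusicEncounterManager by class and instance tag.
; Once bound, it registers with the encounter and forwards tag, name, enum and struct events from the global GameplayEventRouter.
; EndPlay detaches it, uninitializing the encounter when no valid instance tag is set.
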(event EventBeginPlay
  (Networking|CanExecuteCosmeticEvents
    (:True
      (if (MusicEncounterComponent|ShouldInitializeComponent)
        (CallFunction|SendMusicEncounterRequest)))))

(event Custom|Send%20Music%20Encounter%20Request
  (bind _gamestate (Class|FortAsyncActionGameStateReady|GameStateReadyAsync)
    (:OnReady
      (Utilities|IsValid _gamestate
        (:Is%20Valid
          (bind _output_get (Variables|Transient|SetMusic%20Encounter%20Manager (Actor|GetComponentbyClass _gamestate "/MusicEncounter/BP_MusicEncounterManager.BP_MusicEncounterManager_C")))
          (Utilities|IsValid _output_get
            (:Is%20Valid
              (bind _instance_id (MusicEncounterManager|RequestMusicEncounter (Variables|Transient|GetMusic%20Encounter%20Manager) (Variables|MusicEncounterComponent|GameplayTags|GetMusic%20Encounter%20Instance%20Tag) (Variables|MusicEncounterComponent|GetMusic%20Encounter%20Class) (delegate Custom|Handle%20Music%20Encounter%20Ready)))
              (Variables|Transient|SetInstance%20ID _instance_id))))))))

(event Custom|Handle%20Music%20Encounter%20Ready
  (params (Music%20Encounter :type Music%20Encounter%20Base%20Object%20Reference) (Instance%20ID :type Guid%20Structure))
  (if (MusicEncounterComponent|IsRequestedEncounter Music%20Encounter Instance%20ID)
    (MusicEncounterComponent|SetMusicEncounter Music%20Encounter)
    (bind _music_encounter_manager (Variables|Transient|GetMusic%20Encounter%20Manager)
      (:then
        (MusicEncounterManager|UnbindEventfromOnMusicEncounterReady _music_encounter_manager (delegate Custom|Handle%20Music%20Encounter%20Ready)))
      (:else))))

(event Custom|Initialize%20Component
  (CallFunction|StartListeningforEvents))

(event Custom|Uninitialize%20Component
  (CallFunction|StopListeningforEvents))

(event EventEndPlay
  (params (EndPlayReason :type EEndPlayReason%20Enum :default "Destroyed"))
  (Utilities|FlowControl|Sequence
    (:then_0
      (if (not (GameplayTags|IsGameplayTagValid (Variables|MusicEncounterComponent|GameplayTags|GetMusic%20Encounter%20Instance%20Tag)))
        (Class|MusicEncounterBase|UninitializeMusicEncounter (MusicEncounterComponent|GetMusicEncounter))))
    (:then_1
      (MusicEncounterComponent|SetMusicEncounter 0))))

(event Custom|Start%20Listening%20For%20Events
  (Utilities|IsValid (GameplayEventRouter|GetGlobalEventRouter)
    (:Is%20Valid
      (Utilities|FlowControl|Sequence
        (:then_0
          (bind (_asynctaskproxy _payload) (GameplayEventRouter|StartListeningToEvent (GameplayEventRouter|GetGlobalEventRouter) "/MusicEncounter/Events/GameplayEvent_MusicEncounter.GameplayEvent_MusicEncounter" (Components|GetOwner))
            (:then
              (Variables|Transient|SetGameplay%20Event%20Async%20Task _asynctaskproxy))
            (:OnEventReceived
              (bind _payload_gameplaytag_5_9704de5f42b50a6080357e88cd1ba73a (Utilities|Struct|BreakGameplayEventMusicEncounter _payload))
              (MusicEncounter|ExecuteGameplayTagEvent (MusicEncounterComponent|GetMusicEncounter) self _payload_gameplaytag_5_9704de5f42b50a6080357e88cd1ba73a :!class /MusicEncounter/Interfaces/Encounters/BPI_MusicEncounter_GameplayEvent.BPI_MusicEncounter_GameplayEvent_C))))
        (:then_1
          (Utilities|FlowControl|Sequence
            (:then_0
              (bind (_asynctaskproxy_2 _payload_2) (GameplayEventRouter|StartListeningToEvent (GameplayEventRouter|GetGlobalEventRouter) "/MusicEncounter/Events/GameplayEvent_MusicEncounter_Name.GameplayEvent_MusicEncounter_Name" (Components|GetOwner))
                (:then
                  (Variables|Transient|SetNamed%20Gameplay%20Event%20Async%20Task _asynctaskproxy_2))
                (:OnEventReceived
                  (bind _payload_eventname_7_9704de5f42b50a6080357e88cd1ba73a (Utilities|Struct|BreakGameplayEventMusicEncounterName _payload_2))
                  (MusicEncounter|ExecuteNamedEvent (MusicEncounterComponent|GetMusicEncounter) self _payload_eventname_7_9704de5f42b50a6080357e88cd1ba73a :!class /MusicEncounter/Interfaces/Encounters/BPI_MusicEncounter_GameplayEvent.BPI_MusicEncounter_GameplayEvent_C))))
            (:then_1
              (Utilities|FlowControl|Sequence
                (:then_0
                  (bind (_asynctaskproxy_3 _payload_3) (GameplayEventRouter|StartListeningToEvent (GameplayEventRouter|GetGlobalEventRouter) "/MusicEncounter/Events/GameplayEvent_MusicEncounter_Enum.GameplayEvent_MusicEncounter_Enum" (Components|GetOwner))
                    (:then
                      (Variables|Transient|SetEnum%20Gameplay%20Event%20Async%20Task _asynctaskproxy_3))
                    (:OnEventReceived
                      (bind _payload_eventenum_9_9704de5f42b50a6080357e88cd1ba73a (Utilities|Struct|BreakGameplayEventMusicEncounterEnum _payload_3))
                      (MusicEncounter|ExecuteEnumEvent (MusicEncounterComponent|GetMusicEncounter) self _payload_eventenum_9_9704de5f42b50a6080357e88cd1ba73a :!class /MusicEncounter/Encounters/BP_MusicEncounter.BP_MusicEncounter_C))))
                (:then_1
                  (Utilities|FlowControl|Sequence
                    (:then_0
                      (bind (_asynctaskproxy_4 _payload_4) (GameplayEventRouter|StartListeningToEvent (GameplayEventRouter|GetGlobalEventRouter) "/MusicEncounter/Events/GameplayEvent_MusicEncounter_Struct.GameplayEvent_MusicEncounter_Struct" (Components|GetOwner))
                        (:then
                          (Variables|Transient|SetStruct%20Gameplay%20Event%20Async%20Task _asynctaskproxy_4))
                        (:OnEventReceived
                          (bind _payload_eventstruct_11_9704de5f42b50a6080357e88cd1ba73a (Utilities|Struct|BreakGameplayEventMusicEncounterStruct _payload_4))
                          (MusicEncounter|ExecuteStructEvent (MusicEncounterComponent|GetMusicEncounter) self _payload_eventstruct_11_9704de5f42b50a6080357e88cd1ba73a :!class /MusicEncounter/Interfaces/Encounters/BPI_MusicEncounter_GameplayEvent.BPI_MusicEncounter_GameplayEvent_C))))))))))))))

(event Custom|Stop%20Listening%20For%20Events
  (Utilities|FlowControl|Sequence
    (:then_0
      (bind _gameplay_event_async_task (Variables|Transient|GetGameplay%20Event%20Async%20Task)
        (:then
          (GameplayEventRouter|StopListeningtoEvent _gameplay_event_async_task))
        (:else)))
    (:then_1
      (Utilities|FlowControl|Sequence
        (:then_0
          (bind _named_gameplay_event_async_task (Variables|Transient|GetNamed%20Gameplay%20Event%20Async%20Task)
            (:then
              (GameplayEventRouter|StopListeningtoEvent _named_gameplay_event_async_task))
            (:else)))
        (:then_1
          (Utilities|FlowControl|Sequence
            (:then_0
              (bind _enum_gameplay_event_async_task (Variables|Transient|GetEnum%20Gameplay%20Event%20Async%20Task)
                (:then
                  (GameplayEventRouter|StopListeningtoEvent _enum_gameplay_event_async_task))
                (:else)))
            (:then_1
              (Utilities|FlowControl|Sequence
                (:then_0
                  (bind _struct_gameplay_event_async_task (Variables|Transient|GetStruct%20Gameplay%20Event%20Async%20Task)
                    (:then
                      (GameplayEventRouter|StopListeningtoEvent _struct_gameplay_event_async_task))
                    (:else)))))))))))

(fn GetMusicEncounter
  (return (Variables|Transient|GetMusic%20Encounter)))

(fn IsRequestedEncounter
  (params (Music%20Encounter :type Music%20Encounter%20Base%20Object%20Reference) (Instance%20ID :type Guid%20Structure%20%28by%20ref%29))
  ; Ensure encounter class matches
  (Utilities|IsA(soft) Music%20Encounter (Variables|MusicEncounterComponent|GetMusic%20Encounter%20Class)
    (:True
      ; If an instance tag is specified, then ensure tags match
      (if (GameplayTags|IsGameplayTagValid (MusicEncounter|GetMusicEncounterInstanceTag (Variables|Transient|GetMusic%20Encounter)))
        (return (== (MusicEncounter|GetMusicEncounterInstanceTag (Variables|Transient|GetMusic%20Encounter)) (Variables|MusicEncounterComponent|GameplayTags|GetMusic%20Encounter%20Instance%20Tag)))
        (else
          ; Else compare instance ID
          (return (== (Variables|Transient|GetInstance%20ID) (Variables|Transient|GetInstance%20ID))))))
    (:False
      (return false))))

(fn SetMusicEncounter
  (params (Music%20Encounter :type Music%20Encounter%20Base%20Object%20Reference))
  (Utilities|FlowControl|Sequence
    (:then_0
      (bind _music_encounter (Variables|Transient|GetMusic%20Encounter)
        (:then
          (MusicEncounter|RemoveMusicEncounterComponent _music_encounter self)
          (CallFunction|UninitializeComponent))
        (:else)))
    (:then_1
      (bind _output_get (Variables|Transient|SetMusic%20Encounter Music%20Encounter))
      (Utilities|IsValid _output_get
        (:Is%20Valid
          (CallFunction|InitializeComponent)
          (MusicEncounter|AddMusicEncounterComponent (Variables|Transient|GetMusic%20Encounter) self))))))

(fn SetMusicEncounterbyInstanceTag
  (params (Instance%20Tag :type Gameplay%20Tag%20Structure))
  (if (!= Instance%20Tag (Variables|MusicEncounterComponent|GameplayTags|GetMusic%20Encounter%20Instance%20Tag))
    (Variables|MusicEncounterComponent|GameplayTags|SetMusic%20Encounter%20Instance%20Tag (Variables|GetInstance%20Tag))
    (CallFunction|SendMusicEncounterRequest)))

(fn ShouldInitializeComponent
  (Utilities|FlowControl|Sequence
    (:then_0
      (if (Utilities|Array|IsNotEmpty (GameplayTags|BreakGameplayTagContainer (Variables|MusicEncounterComponent|GameplayTags|GetRequired%20Initialization%20Tags)))
        (bind _returnvalue (GameplayTags|HasAllMatchingGameplayTags(Message) (Components|GetOwner) (Variables|MusicEncounterComponent|GameplayTags|GetRequired%20Initialization%20Tags)))
        (LocalVariables|SetHas%20All%20Initialization%20Tags _returnvalue)))
    (:then_1
      (Utilities|FlowControl|Sequence
        (:then_0
          (if (Utilities|Array|IsNotEmpty (GameplayTags|BreakGameplayTagContainer (Variables|MusicEncounterComponent|GameplayTags|GetInitialization%20Blocking%20Tags)))
            (bind _returnvalue_2 (GameplayTags|HasAnyMatchingGameplayTags(Message) (Components|GetOwner) (Variables|MusicEncounterComponent|GameplayTags|GetInitialization%20Blocking%20Tags)))
            (LocalVariables|SetHas%20A%20Blocking%20Tag _returnvalue_2)))
        (:then_1
          (Utilities|FlowControl|Sequence
            (:then_0
              (return (and (LocalVariables|GetHas%20All%20Initialization%20Tags) (not (LocalVariables|GetHas%20A%20Blocking%20Tag)))))))))))
