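; Music encounter component that lets an Audio Gameplay Volume report listener presence to a music encounter.
; On initialization it casts the music encounter to the BPI_MusicEncounter_Volume interface and its owner to an Audio Gameplay Volume, caching both.
; It binds to the volume's listener enter and exit events and forwards each to the encounter through that interface.
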(event EventInitializeComponent
  (|CallParentFunction)
  (bind _asmusic_encounter_volume (Utilities|Casting|CastToBPI_MusicEncounter_Volume (MusicEncounterComponent|GetMusicEncounter)))
  (Variables|MusicEncounter|AudioGameplayVolume|SetAGV%20Interface _asmusic_encounter_volume)
  (bind _asaudio_gameplay_volume (Utilities|Casting|CastToAudioGameplayVolume (Components|GetOwner)))
  (bind _output_get (Variables|MusicEncounter|AudioGameplayVolume|SetAudio%20Gameplay%20Volume _asaudio_gameplay_volume))
  (Events|BindEventtoOnListenerEnterEvent _output_get (EventDispatchers|CreateEvent :!event Custom|Listener%20Entered))
  (Events|BindEventtoOnListenerExitEvent _output_get (EventDispatchers|CreateEvent :!event Custom|Listener%20Exited)))

(event Custom|Listener%20Entered
  (MusicEncounter|Volume|ListenerEntered (Variables|MusicEncounter|AudioGameplayVolume|GetAGV%20Interface)))

(event Custom|Listener%20Exited
  (MusicEncounter|Volume|ListenerExited (Variables|MusicEncounter|AudioGameplayVolume|GetAGV%20Interface)))
