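; Music encounter component that acts as a proxy between an Audio Gameplay Volume and the music encounter's volume interface.
; On initialization it caches the encounter as BPI_MusicEncounter_Volume and finds the AudioGameplayVolumeComponent on its owner.
; If that component is valid, its toggled on and off events are bound to forward ListenerEntered and ListenerExited calls through the cached interface.
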(event EventInitializeComponent
  (bind _returnvalue (Components|GetOwner))
  (bind _returnvalue_2 (Actor|GetComponentbyClass _returnvalue "/Script/AudioGameplayVolume.AudioGameplayVolumeComponent"))
  (|CallParentFunction)
  (bind _asmusic_encounter_volume (Utilities|Casting|CastToBPI_MusicEncounter_Volume (MusicEncounterComponent|GetMusicEncounter)))
  (Variables|MusicEncounter|AudioGameplayVolume|SetAGV%20Interface _asmusic_encounter_volume)
  (Utilities|IsValid _returnvalue_2
    (:Is%20Valid
      (bind _output_get (Variables|MusicEncounter|AudioGameplayVolume|SetAGV%20Proxy%20Component _returnvalue_2))
      (Events|BindEventtoOnToggledOn _output_get (EventDispatchers|CreateEvent :!event Custom|Listener%20Entered))
      (Events|BindEventtoOnToggledOff _output_get (EventDispatchers|CreateEvent :!event Custom|Listener%20Exited)))))

(event Custom|Listener%20Entered
  (MusicEncounter|Volume|ListenerEntered (Variables|MusicEncounter|AudioGameplayVolume|GetAGV%20Interface)))

(event Custom|Listener%20Exited
  (MusicEncounter|Volume|ListenerExited (Variables|MusicEncounter|AudioGameplayVolume|GetAGV%20Interface)))
