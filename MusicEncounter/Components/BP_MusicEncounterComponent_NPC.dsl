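; Music encounter component for NPCs that relays the owning non-participant pawn's state to its music encounter.
; On initialization it caches the owner pawn and casts the encounter to the BPI_MusicEncounter_NPC interface.
; It binds pawn alert level and target changes when those features are enabled, passing in current values, plus death and DBNO changes.
; Each handler forwards its event through that interface, decoupling the NPC from the encounter implementation.
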(event EventInitializeComponent
  (|CallParentFunction)
  (bind _asbp_player_pawn_non_participant (Utilities|Casting|CastToBP_PlayerPawn_NonParticipant (Components|GetOwner)))
  (bind _output_get (Variables|MusicEncounterComponent|References|SetNPC%20Pawn _asbp_player_pawn_non_participant))
  (Utilities|IsValid _output_get
    (:Is%20Valid
      (Utilities|FlowControl|Sequence
        (:then_0
          (bind _asmusic_encounter_npc_interface (Utilities|Casting|CastToBPI_MusicEncounter_NPC (MusicEncounterComponent|GetMusicEncounter)))
          (Variables|MusicEncounterComponent|References|SetNPC%20Interface _asmusic_encounter_npc_interface))
        (:then_1
          (Utilities|FlowControl|Sequence
            (:then_0
              (if (Variables|/Game/Athena/AI/NPCs/Base/BP_PlayerPawn_NonParticipant.BP_PlayerPawn_NonParticipant_C|GetUseAlertness (Variables|MusicEncounterComponent|References|GetNPC%20Pawn))
                (Default|BindEventtoOnAlertLevelChanged (Variables|MusicEncounterComponent|References|GetNPC%20Pawn) (EventDispatchers|CreateEvent :!event Custom|Alert%20Level%20Changed))
                (CallFunction|AlertLevelChanged "Unaware" (Variables|/Game/Athena/AI/NPCs/Base/BP_PlayerPawn_NonParticipant.BP_PlayerPawn_NonParticipant_C|GetAlertLevel (Variables|MusicEncounterComponent|References|GetNPC%20Pawn)))))
            (:then_1
              (Utilities|FlowControl|Sequence
                (:then_0
                  (if (Variables|/Game/Athena/AI/NPCs/Base/BP_PlayerPawn_NonParticipant.BP_PlayerPawn_NonParticipant_C|GetUseChangedTargetReplication (Variables|MusicEncounterComponent|References|GetNPC%20Pawn))
                    (Default|BindEventtoOnTargetChanged (Variables|MusicEncounterComponent|References|GetNPC%20Pawn) (EventDispatchers|CreateEvent :!event Custom|Target%20Changed))
                    (CallFunction|TargetChanged (Variables|/Game/Athena/AI/NPCs/Base/BP_PlayerPawn_NonParticipant.BP_PlayerPawn_NonParticipant_C|GetNewTargetActor (Variables|MusicEncounterComponent|References|GetNPC%20Pawn)))))
                (:then_1
                  (Utilities|FlowControl|Sequence
                    (:then_0
                      (Default|BindEventtoOnDeath (Variables|MusicEncounterComponent|References|GetNPC%20Pawn) (EventDispatchers|CreateEvent :!event Custom|NPC%20Death)))
                    (:then_1
                      (Utilities|FlowControl|Sequence
                        (:then_0
                          (EventDispatchers|BindEventtoOnDBNOStateChanged (Variables|MusicEncounterComponent|References|GetNPC%20Pawn) (EventDispatchers|CreateEvent :!event Custom|NPC%20DBNO%20Changed)))))))))))))))

(event Custom|Alert%20Level%20Changed
  (params (OldAlertLevel :type EAlertLevel%20Enum :default "Unaware") (NewAlertLevel :type EAlertLevel%20Enum :default "Unaware"))
  (MusicEncounter|NPC|AlertLevelChanged (Variables|MusicEncounterComponent|References|GetNPC%20Interface) NewAlertLevel))

(event Custom|Target%20Changed
  (params (NewTarget :type Actor%20Object%20Reference))
  (MusicEncounter|NPC|TargetChanged (Variables|MusicEncounterComponent|References|GetNPC%20Interface) NewTarget))

(event Custom|NPC%20Death
  (params (Cause :type EDeathCause%20Enum :default "OutsideSafeZone"))
  (MusicEncounter|NPC|NPCDeath (Variables|MusicEncounterComponent|References|GetNPC%20Interface) (Variables|MusicEncounterComponent|References|GetNPC%20Pawn)))

(event Custom|NPC%20DBNO%20Changed
  (params (FortPawn :type Fort%20Pawn%20Object%20Reference) (bInIsDBNO :type Boolean))
  (MusicEncounter|NPC|NPCDBNOChanged (Variables|MusicEncounterComponent|References|GetNPC%20Interface) (Variables|MusicEncounterComponent|References|GetNPC%20Pawn) bInIsDBNO))
