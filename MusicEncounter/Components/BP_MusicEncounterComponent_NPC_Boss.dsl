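; Boss NPC music encounter component that relays the owning boss pawn's state to its music encounter through interface messages.
; On initialize it caches the boss pawn and binds health and conversion events; overrides forward death, DBNO, alert level and target changes.
; A delta gate reports changes in locally viewed player targeting, while Unaware alerts force it false as a bug fix.
; Unconversion is forwarded, after an encounter request, only for unexcluded reasons.
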(event EventInitializeComponent
  (|CallParentFunction)
  (bind _asbp_player_pawn_non_participant_boss (Utilities|Casting|CastToBP_PlayerPawn_NonParticipant_Boss (Components|GetOwner)))
  (bind _output_get (Variables|MusicEncounterComponent|Boss|SetBoss%20Pawn _asbp_player_pawn_non_participant_boss))
  (Utilities|IsValid _output_get
    (:Is%20Valid
      (Default|BindEventtoOnHealthChange _output_get (EventDispatchers|CreateEvent :!event Custom|Boss%20Health%20Changed))
      (bind _convertcomponent (Variables|/Script/FortniteGame.FortPlayerPawn|GetConvertComponent (Variables|MusicEncounterComponent|Boss|GetBoss%20Pawn))
        (:then
          (ConvertEvents|BindEventtoOnConvertedEvent _convertcomponent (EventDispatchers|CreateEvent :!event Custom|On%20Boss%20Converted))
          (ConvertEvents|BindEventtoOnUnconvertedEvent _convertcomponent (EventDispatchers|CreateEvent :!event On%20Boss%20UnConverted)))
        (:else)))))

(event Custom|Boss%20Health%20Changed
  (bind _boss_pawn (Variables|MusicEncounterComponent|Boss|GetBoss%20Pawn)
    (:then
      (MusicEncounter|Boss|BossHealthChanged(Message) (MusicEncounterComponent|GetMusicEncounter) self (Variables|/Game/Athena/AI/NPCs/Base/BP_PlayerPawn_NonParticipant_Boss.BP_PlayerPawn_NonParticipant_Boss_C|GetHealthInfo _boss_pawn)))
    (:else)))

(event EventNPCDeath
  (params (Cause :type EDeathCause%20Enum :default "OutsideSafeZone"))
  (MusicEncounter|Boss|BossDeath(Message) (MusicEncounterComponent|GetMusicEncounter) self))

(event EventAlertLevelChanged
  (params (OldAlertLevel :type EAlertLevel%20Enum :default "Unaware") (NewAlertLevel :type EAlertLevel%20Enum :default "Unaware"))
  (MusicEncounter|Boss|BossAlertLevelChanged(Message) (MusicEncounterComponent|GetMusicEncounter) self NewAlertLevel)
  (if (Utilities|Enum|Equal(Enum) NewAlertLevel "Unaware")
    (MusicEncounter|Boss|IsBossTargetingListenerChanged(Message) (MusicEncounterComponent|GetMusicEncounter) self false) ; Fixes an issue where NPC can still target a player while Unaware
  ))

(event EventTargetChanged
  (params (NewTarget :type Actor%20Object%20Reference))
  (MusicEncounter|Boss|BossTargetChanged(Message) (MusicEncounterComponent|GetMusicEncounter) self NewTarget)
  (bind _asfort_player_pawn_athena (Utilities|Casting|CastToFortPlayerPawnAthena NewTarget)
    (:then
      (Utilities|IsValid _asfort_player_pawn_athena
        (:Is%20Valid
          (if (Pawn|IsLocallyViewed _asfort_player_pawn_athena)
            (MusicEncounterComponent|Boss|SetIsListenerTargeted true)
            (else
              _)))))
    (:CastFailed
      _))
  (MusicEncounterComponent|Boss|SetIsListenerTargeted false))

(event EventNPCDBNOChanged
  (params (FortPawn :type Fort%20Pawn%20Object%20Reference) (bInIsDBNO :type Boolean))
  (MusicEncounter|Boss|BossDBNOChanged(Message) (MusicEncounterComponent|GetMusicEncounter) self bInIsDBNO))

(event Custom|On%20Boss%20Converted
  (params (InstigatorPawn :type Fort%20Pawn%20Object%20Reference) (ConvertedPawn :type Fort%20Pawn%20Object%20Reference))
  (MusicEncounter|Boss|BossConversionStateChanged(Message) (MusicEncounterComponent|GetMusicEncounter) self InstigatorPawn ConvertedPawn true))

(event Custom|On%20Boss%20Unconverted
  (params (UnconvertedPawn :type Fort%20Pawn%20Object%20Reference) (UnconvertReason :type EUnconvertReason%20Enum :default "StealByOtherPlayer"))
  (if (Math|Boolean|ANDBoolean (Utilities|Enum|NotEqual(Enum) UnconvertReason "ConvertedDeath") (Utilities|Enum|NotEqual(Enum) UnconvertReason "ConverterDeath") (Utilities|Enum|NotEqual(Enum) UnconvertReason "StealByOtherPlayer") (Utilities|Enum|NotEqual(Enum) UnconvertReason "EndPlay") (Utilities|Enum|NotEqual(Enum) UnconvertReason "GiveOwnershipToSquadmate"))
    (CallFunction|SendMusicEncounterRequest)
    (MusicEncounter|Boss|BossConversionStateChanged(Message) (MusicEncounterComponent|GetMusicEncounter) self 0 UnconvertedPawn false)))

(fn SetIsListenerTargeted
  (params (Is%20Listener%20Targeted :type Boolean))
  (bind _updated_variable (Utilities|FlowControl|DeltaGateWildcard (Variables|MusicEncounterComponent|Boss|GetIs%20Listener%20Targeted) Is%20Listener%20Targeted)
    (:Success
      (MusicEncounter|Boss|IsBossTargetingListenerChanged(Message) (MusicEncounterComponent|GetMusicEncounter) self _updated_variable))))

(fn UsesTargetChangedReplication
  (bind _boss_pawn (Variables|MusicEncounterComponent|Boss|GetBoss%20Pawn)
    (:then
      (return (Variables|/Game/Athena/AI/NPCs/Base/BP_PlayerPawn_NonParticipant.BP_PlayerPawn_NonParticipant_C|GetUseChangedTargetReplication _boss_pawn)))
    (:else)))
